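# Niagara system for the body aura of the EarPsych emote, pinning a pulsing cyan-to-magenta glow to the player's skeleton.
# Mobile_Aura places bone-attached sprites with per-bone sizes from arrays indexed by spawn group.
# Texture bursts partitioned, bone-oriented meshes, while long_exposure_effect continuously spawns short-lived ones at a rate driven by User.Body_Aura_SpawnRate.
# Time-of-day and front-end modifiers adjust brightness, and front-end state also sets mesh visibility tags.
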NiagaraSystem /BRCosmetics/Animation/Game/MainPlayer/Emotes/EarPsych/FX/NS_Emote_EarPsych_Aura
dependencies {
 usedDataInterfaces [
  &NiagaraDataInterfaceColorCurve
  &NiagaraDataInterfaceCurve
  &NiagaraDataInterfaceArrayFloat
  &NiagaraDataInterfaceSkeletalMesh
  &NiagaraDataInterfaceArrayFloat3
 ]
 usedDynamicInputs [&RandomRangeLinearColor]
 usedModules [
  &self:SystemSpawnScript.SystemScriptSource.SystemScriptGraph.NiagaraNodeAssignment_0.SetVariables_#1
  &SystemState
  &self:SystemSpawnScript.SystemScriptSource.SystemScriptGraph.NiagaraNodeAssignment_1.SetVariables_#2
 ]
}
emitters {
 Mobile_Aura {
  emitterUpdateScript [
   &EmitterState(Life Cycle Mode=::NewEnumerator1, Loop Behavior=::NewEnumerator1, Loop Duration=1.5, UseLoopDelay=true)
   &SpawnBurst_Instantaneous(Spawn Count=13, Use Spawn Probability=false, UseLoopCountLimit=false)
  ]
  particleSpawnScript [
   &FN_SetSpawnGroup(Number of Spawn Groups=15)
   &InitializeParticle(
    Color Mode=::NewEnumerator0
    Lifetime=3.5
    Sprite Size Mode=::NewEnumerator3
    Sprite UV Mode=::NewEnumerator1
    Uniform Sprite Size=&SelectFloatFromArray(
     Direct Array Index=$Particles.SpawnGroup
     Float Selection Array=DataInterface() {propertyValues {FloatData [100,70,70,70,70,50,50,130,130,100,100,100,100]}}
    )
    UsePositionOffset=false
    Write Material Random=true
   )
   &CameraOffset(Camera Offset Amount=-10)
  ]
  particleUpdateScript [
   &ParticleState()
   &SkeletalMeshLocation(
    Bone / Socket Index=$Particles.SpawnGroup
    Bone Sampling Mode=::NewEnumerator4
    MaskSpawnGroup=false
    Sampled Position Offset=&SelectVectorFromArray(
     Direct Array Index=$Particles.SpawnGroup
     Vector Selection Array=DataInterface() {propertyValues {FloatData [<0 0 0>, <0 0 0>, <0 0 0>, <0 0 0>, <0 0 0>, <5 0 0>, <-5 0 0>, <0 0 0>, <0 0 0>, <0 0 0>, <0 0 0>, <0 0 0>, <0 0 0>]}}
    )
    Skeletal Mesh=DataInterface() {propertyValues {
      DefaultMesh=null
      ExcludeBoneName=root
      OptionalComponentClass=null
      PreviewMesh=&M_Med_Soldier_04_Preview
      SkinningMode=SkinOnTheFly
      SoftSourceActor=null
      SourceMode=Default
      UvSetIndex=0
      WholeMeshLOD=-1
      bExcludeBone=false
      bReadDeformedGeometry=true
      bRequireCurrentFrameData=true
      ComponentTags []
      FilteredBones [head,upperarm_l,upperarm_r,lowerarm_l,lowerarm_r,hand_l,hand_r,thigh_l,thigh_r,calf_l,calf_r,foot_l,foot_r]
      FilteredSockets []
      MeshUserParameter {parameter {name=None, typeDefHandle {registeredTypeIndex=1}, varData []}}
      SamplingRegions []
     }}
   )
   &ScaleSpriteSize(
    Scale Factor=&Multiply_Vector2D_ByFloat(
     Float=&Waveform("Waveform [1]"=::NewEnumerator0, "[1] Frequency"=1, "[1] Amplitude Min/Max"=<0.6 1>)
     Vector2D=&RandomRangeVector2D(Minimum=<1 1>, Maximum=<1 1>)
    )
    Scale Sprite Size Mode=::NewEnumerator2
   ) {module=ScaleSpriteSize001}
   &DynamicMaterialParameters(
    Index 0 Param 4=&RandomRangeFloat(Minimum=0, Maximum=1)
    Module.Param0WriteEnabled=true
    Module.Param1WriteEnabled=true
    Module.Param2WriteEnabled=true
    Module.Param3WriteEnabled=true
   )
   &ScaleColor(
    Scale Alpha=&Multiply_Float(
     A=&Multiply_Float(A=1, B=&FN_FrontEndFloatModifier(Gameplay=1, Front End=1))
     B=&FloatFromCurve(
      FloatCurve=DataInterface() {propertyValues {Curve {keys [
          {tangentMode=RCTM_User, time=0, value=0}
          <at 0.2 1>
          {arriveTangent=-0.0405, interpMode=RCIM_Cubic, leaveTangent=-0.0405, tangentMode=RCTM_User, time=0.7009, value=1}
          {arriveTangent=-8.0089, interpMode=RCIM_Cubic, leaveTangent=-8.0089, tangentMode=RCTM_User, time=1, value=0}
         ]}}}
      Scale Curve=&FN_FrontEndFloatModifier(
       Gameplay=&FNTimeOfDayScalar(Night Scale=0.2, Morning Scale=0.5, Day Scale=0.5, Evening Scale=0.5)
       Front End=0.5
      )
     )
    )
    Scale RGB=&Multiply_VectorByFloat(
     Vector=&MakeVectorFromLinearColorRGB(Color=&ColorFromCurve(ColorCurve=DataInterface() {propertyValues {
         AlphaCurve {keys [<at 0 1 RCIM_Cubic>,<at 1 0 RCIM_Cubic>]}
         BlueCurve {keys [<at 0 1 RCIM_Cubic>,<at 1 1 RCIM_Cubic>]}
         GreenCurve {keys [<at 0 0.9123 RCIM_Cubic>,<at 1 0 RCIM_Cubic>]}
         RedCurve {keys [<at 0 0 RCIM_Cubic>,<at 1 0.8796 RCIM_Cubic>]}
        }}))
     Float=&Multiply_Float(A=$System.TOD, B=&FN_FrontEndFloatModifier(Gameplay=1, Front End=1))
    )
    ScaleA=true
    ScaleRGB=true
   )
  ]
  rendererClasses [&NiagaraSpriteRendererProperties]
  renderers [
   {rendererClass=&NiagaraSpriteRendererProperties}
   {rendererClass=&NiagaraSpriteRendererProperties, rendererIndex=1}
  ]
 }
 Texture {
  emitterUpdateScript [
   /SetVariables(Emitter.OutlineThickness:NiagaraFloat=&Waveform("Waveform [1]"=::NewEnumerator0, "[1] Frequency"=1, "[1] Amplitude Min/Max"=<0 1>)) {isSetParameters=true, module=SetVariables_#3}
   /SetVariables(Emitter.GlowOutlineThickness:NiagaraFloat=&Waveform(Global Amplitude Scale=7, "Waveform [1]"=::NewEnumerator0, "[1] Frequency"=1, "[1] Amplitude Min/Max"=<0 1>)) {isSetParameters=true, module=SetVariables_#4}
   /SetVariables(
    Emitter.Color a:LinearColor=&ColorFromCurve(
     ColorCurve=DataInterface() {propertyValues {
       AlphaCurve {keys [<at 0 1 RCIM_Cubic>,<at 1 0 RCIM_Cubic>]}
       BlueCurve {keys [<at 0 1 RCIM_Cubic>,<at 1 1 RCIM_Cubic>]}
       GreenCurve {keys [<at 0 0.9123 RCIM_Cubic>,<at 1 0 RCIM_Cubic>]}
       RedCurve {keys [<at 0 0 RCIM_Cubic>,<at 1 0.8796 RCIM_Cubic>]}
      }}
     CurveIndex=$Emitter.NormalizedLoopAge
    )
    Emitter.color b:LinearColor=&ColorFromCurve(
     ColorCurve=DataInterface() {propertyValues {
       AlphaCurve {keys [<at 0 1 RCIM_Cubic>,<at 1 0 RCIM_Cubic>]}
       BlueCurve {keys [<at 0 1 RCIM_Cubic>,<at 1 1 RCIM_Cubic>]}
       GreenCurve {keys [<at 0 0.9123 RCIM_Cubic>,<at 1 0 RCIM_Cubic>]}
       RedCurve {keys [<at 0 0 RCIM_Cubic>,<at 1 0.8796 RCIM_Cubic>]}
      }}
     CurveIndex=$Emitter.NormalizedLoopAge
    )
    Emitter.fe color:LinearColor=&MakeLinearColorFromVectorAndFloat(
     "Vector (RGB)"=&MakeVectorFromLinearColorRGB(Color=&ColorFromCurve(
       ColorCurve=DataInterface() {propertyValues {
         AlphaCurve {keys [<at 0 1 RCIM_Cubic>,<at 1 0 RCIM_Cubic>]}
         BlueCurve {keys [<at 0 1 RCIM_Cubic>,<at 1 1 RCIM_Cubic>]}
         GreenCurve {keys [<at 0 0.9123 RCIM_Cubic>,<at 1 0 RCIM_Cubic>]}
         RedCurve {keys [<at 0 0 RCIM_Cubic>,<at 1 0.8796 RCIM_Cubic>]}
        }}
       CurveIndex=$Emitter.NormalizedLoopAge
      ))
     "Float (Alpha)"=1
    )
   ) {isSetParameters=true, module=SetVariables_#5}
   &EmitterState(Inactive Response=::NewEnumerator1, Life Cycle Mode=::NewEnumerator1, Loop Behavior=::NewEnumerator1, Loop Duration=3.5, UseLoopDelay=false)
   &SpawnBurst_Instantaneous(Spawn Group=1, Spawn Time=0.5, Use Spawn Probability=false, UseLoopCountLimit=false) {module=SpawnBurst_Instantaneous002}
   &SpawnBurst_Instantaneous(Spawn Count=7, Spawn Time=0.5, Use Spawn Probability=false, UseLoopCountLimit=false)
  ]
  particleSpawnScript [
   /SetVariables(Particles.VisibilityTag=&MakeIntFromFloat(Float=&FN_FrontEndFloatModifier(Gameplay=0, Front End=1))) {isSetParameters=true, module=SetVariables_#6}
   &InitializeParticle(
    Color Mode=::NewEnumerator0
    Lifetime=3
    Mesh Scale=<20 20 30>
    Mesh Scale Mode=::NewEnumerator4
    UsePositionOffset=true
    Write Material Random=true
   )
   &PartitionParticles(Number Of Partitions=8)
  ]
  particleUpdateScript [
   &ParticleState()
   &ScaleMeshSize(Scale Factor=&VectorFromFloat(Value=0.85))
   &ScaleColor(
    Scale Alpha=&FloatFromCurve(
     FloatCurve=DataInterface() {propertyValues {Curve {keys [
         {arriveTangent=13.6388, interpMode=RCIM_Cubic, time=0.003, value=0.0083}
         {arriveTangent=15.3759, leaveTangent=15.3759, time=0.1, value=1.5}
         <at 0.3 1.5>
         {interpMode=RCIM_Cubic, tangentMode=RCTM_User, time=0.6756, value=1.5}
         {interpMode=RCIM_Cubic, leaveTangent=-1.6667, time=1, value=0}
        ]}}}
     Scale Curve=&FN_FrontEndFloatModifier(
      Gameplay=&FNTimeOfDayScalar(Night Scale=0.25, Morning Scale=0.3, Day Scale=0.45, Evening Scale=0.5)
      Front End=1
     )
    )
    ScaleA=true
    ScaleRGB=true
   ) {module=ScaleColor001}
   &SkeletalMeshLocation(
    Bone / Socket Index=$Particles.PartitionParticles.Partition
    Bone Sampling Mode=::NewEnumerator4
    MaskSpawnGroup=false
    Orientation Sampling=::NewEnumerator1
    Sampled Position Offset=<0 0 5>
    Skeletal Mesh=$System.Skeletal Mesh
   )
   &CameraOffset(Camera Offset Amount=20)
   &DynamicMaterialParameters(Module.Index0Float0Write=true, Module.Index0Float1Write=true, Module.Index0Float2Write=true, Module.Index0Float3Write=true)
   /SetVariables(Particles.squares:Vector3f=&Multiply_VectorByFloat(Vector=<0.5 0.5 1>, Float=5.5879)) {isSetParameters=true}
   /SetVariables(Particles.FE Color:LinearColor=&MakeLinearColorFromVectorAndFloat(
     "Vector (RGB)"=&MakeVectorFromLinearColorRGB(Color=&ColorFromCurve(
       ColorCurve=DataInterface() {propertyValues {
         AlphaCurve {keys [<at 0 1 RCIM_Cubic>,<at 1 0 RCIM_Cubic>]}
         BlueCurve {keys [<at 0 1 RCIM_Cubic>,<at 1 1 RCIM_Cubic>]}
         GreenCurve {keys [<at 0 0.9123 RCIM_Cubic>,<at 1 0 RCIM_Cubic>]}
         RedCurve {keys [<at 0 0 RCIM_Cubic>,<at 1 0.8796 RCIM_Cubic>]}
        }}
       CurveIndex=$Emitter.NormalizedLoopAge
      ))
     "Float (Alpha)"=0.2
    )) {isSetParameters=true}
  ]
  rendererClasses [&NiagaraMeshRendererProperties]
  renderers [
   {rendererClass=&NiagaraMeshRendererProperties, rendererIndex=0}
   {rendererClass=&NiagaraMeshRendererProperties, rendererIndex=1}
  ]
 }
 long_exposure_effect {
  emitterUpdateScript [
   /SetVariables(Emitter.OutlineThickness:NiagaraFloat=&Waveform("Waveform [1]"=::NewEnumerator0, "[1] Frequency"=1, "[1] Amplitude Min/Max"=<0 1>)) {isSetParameters=true, module=SetVariables_#3}
   /SetVariables(Emitter.GlowOutlineThickness:NiagaraFloat=&Waveform(
     Global Amplitude Scale=&FN_FrontEndFloatModifier(Gameplay=7, Front End=0.5)
     "Waveform [1]"=::NewEnumerator0
     "[1] Frequency"=1
     "[1] Amplitude Min/Max"=<0 1>
    )) {isSetParameters=true, module=SetVariables_#4}
   /SetVariables(
    Emitter.Color a:LinearColor=&ColorFromCurve(
     ColorCurve=DataInterface() {propertyValues {
       AlphaCurve {keys [<at 0 1 RCIM_Cubic>,<at 1 0 RCIM_Cubic>]}
       BlueCurve {keys [<at 0 1 RCIM_Cubic>,<at 1 1 RCIM_Cubic>]}
       GreenCurve {keys [<at 0 0.9123 RCIM_Cubic>,<at 1 0 RCIM_Cubic>]}
       RedCurve {keys [<at 0 0 RCIM_Cubic>,<at 1 0.8796 RCIM_Cubic>]}
      }}
     CurveIndex=$Emitter.NormalizedLoopAge
    )
    Emitter.color b:LinearColor=&ColorFromCurve(
     ColorCurve=DataInterface() {propertyValues {
       AlphaCurve {keys [<at 0 0.5 RCIM_Cubic>,<at 1 0 RCIM_Cubic>]}
       BlueCurve {keys [<at 0 1 RCIM_Cubic>,<at 1 1 RCIM_Cubic>]}
       GreenCurve {keys [<at 0 0.9123 RCIM_Cubic>,<at 1 0 RCIM_Cubic>]}
       RedCurve {keys [<at 0 0 RCIM_Cubic>,<at 1 0.8796 RCIM_Cubic>]}
      }}
     CurveIndex=$Emitter.NormalizedLoopAge
    )
   ) {isSetParameters=true, module=SetVariables_#5}
   &EmitterState(Life Cycle Mode=::NewEnumerator1, Loop Behavior=::NewEnumerator1, Loop Duration=3.5, UseLoopDelay=false)
   &SpawnRate(SpawnGroup=1, SpawnRate=&Lerp_Float(A=0, B=35, Alpha=$User.Body_Aura_SpawnRate), Use Spawn Probability=false) {module=SpawnRate001}
   &SpawnRate(SpawnRate=&Lerp_Float(A=0, B=35, Alpha=$User.Body_Aura_SpawnRate), Use Spawn Probability=false)
   &SpawnBurst_Instantaneous(Spawn Count=7, Spawn Time=0.2, Use Spawn Probability=false, UseLoopCountLimit=false) {enabled=false, module=SpawnBurst_Instantaneous001}
   &SpawnBurst_Instantaneous(Spawn Count=7, Spawn Time=0.5, Use Spawn Probability=false, UseLoopCountLimit=false) {enabled=false}
  ]
  particleSpawnScript [
   /SetVariables(Particles.VisibilityTag=&MakeIntFromFloat(Float=&FN_FrontEndFloatModifier(Gameplay=0, Front End=1))) {isSetParameters=true, module=SetVariables_#6}
   &InitializeParticle(Color Mode=::NewEnumerator0, Lifetime=0.5, Mesh Scale Mode=::NewEnumerator4, UsePositionOffset=true, Write Material Random=true)
   &PartitionParticles(Number Of Partitions=7)
   /SetVariables() {enabled=false, isSetParameters=true}
  ]
  particleUpdateScript [
   &ParticleState()
   &ScaleMeshSize(Scale Factor=&VectorFromFloat(Value=0.8))
   &ScaleColor(
    Scale Alpha=&FloatFromCurve(
     FloatCurve=DataInterface() {propertyValues {Curve {keys [
         {arriveTangent=13.6388, interpMode=RCIM_Cubic, time=0.003, value=0.0083}
         {arriveTangent=15.3759, leaveTangent=15.3759, time=0.1, value=1.5}
         <at 0.3 1.5>
         {interpMode=RCIM_Cubic, tangentMode=RCTM_User, time=0.6756, value=1.5}
         {interpMode=RCIM_Cubic, leaveTangent=-1.6667, time=1, value=0}
        ]}}}
     Scale Curve=&FNTimeOfDayScalar(Night Scale=1.5, Morning Scale=1, Day Scale=1.3, Evening Scale=1.5)
    )
    ScaleA=true
    ScaleRGB=true
   ) {module=ScaleColor001}
   &SkeletalMeshLocation(
    Bone / Socket Index=$Particles.PartitionParticles.Partition
    Bone Sampling Mode=::NewEnumerator4
    MaskSpawnGroup=false
    Orientation Sampling=::NewEnumerator1
    Skeletal Mesh=$System.Skeletal Mesh
   )
   &CameraOffset(Camera Offset Amount=20)
   &DynamicMaterialParameters(Module.Index0Float0Write=true, Module.Index0Float1Write=true, Module.Index0Float2Write=true, Module.Index0Float3Write=true)
   /SetVariables(Particles.FE Color:LinearColor=&MakeLinearColorFromVectorAndFloat(
     "Vector (RGB)"=&MakeVectorFromLinearColorRGB(Color=&ColorFromCurve(
       ColorCurve=DataInterface() {propertyValues {
         AlphaCurve {keys [<at 0 1 RCIM_Cubic>,<at 1 0 RCIM_Cubic>]}
         BlueCurve {keys [<at 0 1 RCIM_Cubic>,<at 1 1 RCIM_Cubic>]}
         GreenCurve {keys [<at 0 0.9123 RCIM_Cubic>,<at 1 0 RCIM_Cubic>]}
         RedCurve {keys [<at 0 0 RCIM_Cubic>,<at 1 0.8796 RCIM_Cubic>]}
        }}
       CurveIndex=$Emitter.NormalizedLoopAge
      ))
     "Float (Alpha)"=0.5
    )) {isSetParameters=true}
  ]
  rendererClasses [&NiagaraMeshRendererProperties]
  renderers [
   {rendererClass=&NiagaraMeshRendererProperties, rendererIndex=0}
   {rendererClass=&NiagaraMeshRendererProperties, rendererIndex=1}
  ]
 }
}
summary {
 systemName=NS_Emote_EarPsych_Aura
 emitters [
  {emitterName=long_exposure_effect, rendererClasses [&NiagaraMeshRendererProperties]}
  {emitterName=Mobile_Aura, rendererClasses [&NiagaraSpriteRendererProperties]}
  {emitterName=Texture, rendererClasses [&NiagaraMeshRendererProperties]}
 ]
 userVariables [
  User.bIsFrontEnd {type=NiagaraBool}
  User.bIsFrontEndPreview {type=NiagaraBool}
  User.BlinkIndex {type=NiagaraFloat, defaultValue=0, description=""}
  User.Body_Aura_SpawnRate {type=NiagaraFloat, defaultValue=1, description=""}
 ]
}
