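# Niagara floor effect for the EarPsych emote, pulsing expanding meshes and a glow sprite at a skeletal mesh sample point.
# Twin FE and Gameplay emitters burst two spawn groups per loop, with lifetimes zeroed outside their front-end or gameplay context.
# Curves drive scale, dissolve-style material parameters and glow, pixel, scan and grid colors, with alphas scaled by time of day.
# Gameplay resamples position every update and traces onto ground slope; its BlendSpace module is disabled.
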NiagaraSystem /BRCosmetics/Animation/Game/MainPlayer/Emotes/EarPsych/FX/NS_Emote_EarPsych_floor
dependencies {
 usedDataInterfaces [
  &NiagaraDataInterfaceCurve
  &NiagaraDataInterfaceCollisionQuery
  &NiagaraDataInterfaceDebugDraw
  &NiagaraDataInterfaceSkeletalMesh
 ]
 usedDynamicInputs [&Lerp_Float]
 usedModules [
  &self:SystemSpawnScript.SystemScriptSource.SystemScriptGraph.NiagaraNodeAssignment_0.SetVariables_#1
  &SystemState
  &BlendSpaceHelper
 ]
}
emitters {
 FE {
  emitterSpawnScript [/SetVariables(Emitter.Size=0.5) {isSetParameters=true, module=SetVariables_#2}]
  emitterUpdateScript [
   &EmitterState(Life Cycle Mode=::NewEnumerator1, Loop Duration=0.5, Recalculate Duration Each Loop=true, UseLoopDelay=false)
   &SpawnBurst_Instantaneous(Spawn Group=1, Use Spawn Probability=false, UseLoopCountLimit=false)
   &SpawnBurst_Instantaneous(Use Spawn Probability=false, UseLoopCountLimit=false) {module=SpawnBurst_Instantaneous001}
  ]
  particleSpawnScript [
   /SetVariables(Particles.VisibilityTag=&SpawnGroup()) {isSetParameters=true, module=SetVariables_#3}
   &InitializeParticle(
    Color=<rgba 0.5616 0.6 0.5671 1>
    Lifetime=&MergeHelper_FloatSwitch(
     0=&FN_FrontEndFloatModifier(Gameplay=0, Front End=1)
     1=&FN_FrontEndFloatModifier(Gameplay=0, Front End=0.4)
     2=?
     3=?
     Index=&SpawnGroup()
     Num=::NewEnumerator0
    )
    Mesh Scale Mode=::NewEnumerator3
    Mesh Uniform Scale=0.7
    Position Offset=<0 1 0>
    UsePositionOffset=false
    Write Material Random=true
   )
   /SetVariables(
    Particles.Glow Lifetime=0.3
    Particles.Glow Size=&Multiply_Vector2D_ByFloat(Float=$Emitter.Size, Vector2D=&Multiply_Vector2D_ByFloat(Float=80, Vector2D=<1 1>))
    Particles.GlowFacing:Vector3f=<0 0 1>
   ) {isSetParameters=true, module=SetVariables_#4}
   &SkeletalMeshLocation(
    Bone Sampling Mode=::NewEnumerator4
    MaskSpawnGroup=false
    Orientation Sampling=::NewEnumerator1
    Sampled Position Offset=<0 0 2>
    Sampled Position Offset Coordinate Space:ENiagaraSkelMeshTransforms=::NewEnumerator2
    Skeletal Mesh=$System.Skeletal Mesh Reader
   )
   &InitialMeshOrientation(
    Mesh Axis To Orient=<0 0 1>
    Mesh Orientation Mode=::NewEnumerator2
    Orientation Coordinate Space=::Local
    Orientation Vector=&FN_FrontEndVector3Modifier(Gameplay=<0 0 1>, Front End=<0 0 1>)
    Rotation=<5 0 0>
    Rotation Coordinate Space=::NewEnumerator0
    Use Rotation Vector=false
   )
   /Game/Sprite/FN_SpriteAlignmentAndFacing(SpriteAlignment:Vector3f=<0 1 0>, SpriteFacing:Vector3f=<0 0 1>)
  ]
  particleUpdateScript [
   &ParticleState()
   /SetVariables(Particles.Update Glow Age:NiagaraFloat=&Divide_Float(A=$Particles.Age, B=$Particles.Glow Lifetime)) {isSetParameters=true, module=SetVariables_#5}
   &ScaleMeshSize(Scale Factor=&VectorFromFloat(Value=&FloatFromCurve(
      FloatCurve=DataInterface() {propertyValues {Curve {keys [
          {arriveTangent=1.0263, interpMode=RCIM_Cubic, leaveTangent=1.0263, tangentMode=RCTM_User, time=0, value=1}
          {arriveTangent=0.4145, interpMode=RCIM_Cubic, leaveTangent=0.4145, tangentMode=RCTM_User, time=0.2, value=1.375}
          {arriveTangent=0.018, interpMode=RCIM_Cubic, leaveTangent=0.018, tangentMode=RCTM_User, time=0.9945, value=1.5}
         ]}}}
      Scale Curve=0.7
     )))
   /SetVariables(
    Particles.Update Glow Color:LinearColor=&MakeLinearColorFromVectorAndFloat(
     "Vector (RGB)"=&MakeVectorFromLinearColorRGB(Color=<rgba 0 1 0.6099 1>)
     "Float (Alpha)"=&FloatFromCurve(
      FloatCurve=DataInterface() {propertyValues {Curve {
         defaultValue=1
         keys [
          {interpMode=RCIM_Cubic, tangentMode=RCTM_User, time=0, value=0}
          {interpMode=RCIM_Cubic, tangentMode=RCTM_User, time=0.0724, value=1}
          {tangentMode=RCTM_User, time=1, value=0}
         ]
        }}}
      CurveIndex=$Particles.Update Glow Age
      Scale Curve=&FNTimeOfDayScalar(Night Scale=0.1, Morning Scale=0.1, Day Scale=0.15, Evening Scale=0.1)
     )
    )
    Particles.Update Glow DynamicParameter:Vector4f=<2 0 0 0>
    Particles.Update Glow Size:Vector2f=&Multiply_Vector2D(
     A=$Particles.Glow Size
     B=&Multiply_Vector2D_ByFloat(
      Float=&FloatFromCurve(
       FloatCurve=DataInterface() {propertyValues {Curve {
          defaultValue=0.367
          keys [
           {interpMode=RCIM_Cubic, tangentMode=RCTM_User, time=0, value=0.367}
           {arriveTangent=-0.1298, interpMode=RCIM_Cubic, leaveTangent=-0.1298, tangentMode=RCTM_User, time=1, value=3.8846}
          ]
         }}}
       CurveIndex=$Particles.Update Glow Age
       Scale Curve=1.5
      )
      Vector2D=<1.5 1.5>
     )
    )
   ) {isSetParameters=true, module=SetVariables_#6}
   /SetVariables(Particles.Scale002:Vector3f=&Multiply_Vector(
     A=<0.5 0.5 0.1>
     B=&VectorFromFloat(Value=&FloatFromCurve(FloatCurve=DataInterface() {propertyValues {Curve {keys [
           {arriveTangent=1.0263, interpMode=RCIM_Cubic, leaveTangent=1.0263, tangentMode=RCTM_User, time=0, value=1}
           {arriveTangent=0.3981, interpMode=RCIM_Cubic, leaveTangent=0.3981, tangentMode=RCTM_User, time=0.2, value=1.375}
           {arriveTangent=0.018, interpMode=RCIM_Cubic, leaveTangent=0.018, tangentMode=RCTM_User, time=0.9945, value=1.5}
          ]}}}))
    )) {isSetParameters=true, module=SetVariables_#7}
   &DynamicMaterialParameters(
    Index 0 Param 1=&FloatFromCurve(FloatCurve=DataInterface() {propertyValues {Curve {keys [
         {arriveTangent=-4.087, interpMode=RCIM_Cubic, leaveTangent=-4.087, tangentMode=RCTM_User, time=0, value=0.5}
         {arriveTangent=-1.721, interpMode=RCIM_Cubic, leaveTangent=-1.721, tangentMode=RCTM_User, time=0.1324, value=-0.0129}
         {arriveTangent=-0.376, interpMode=RCIM_Cubic, leaveTangent=-0.376, tangentMode=RCTM_User, time=1, value=-1}
        ]}}})
    Module.Index0Param0Write=true
    Module.Index0Param1Write=true
    Module.Index0Param2Write=true
    Module.Index0Param3Write=true
   )
   /SetVariables(Particles.Pixel color:LinearColor=&MakeLinearColorFromVectorAndFloat(
     "Vector (RGB)"=&MakeVectorFromLinearColorRGB(Color=<rgba 0.5 0 0.4489 2>)
     "Float (Alpha)"=&FloatFromCurve(
      FloatCurve=DataInterface() {propertyValues {Curve {keys [
          {arriveTangent=-0.024, interpMode=RCIM_Cubic, leaveTangent=-0.024, tangentMode=RCTM_User, time=0, value=0}
          {arriveTangent=-0.3677, interpMode=RCIM_Cubic, leaveTangent=-0.3677, tangentMode=RCTM_User, time=0.1, value=1}
          {arriveTangent=-1.5833, interpMode=RCIM_Cubic, leaveTangent=-1.5833, time=0.3, value=0.4}
          {arriveTangent=-0.3356, interpMode=RCIM_Cubic, leaveTangent=-0.3356, tangentMode=RCTM_User, time=0.7, value=0.05}
          <at 1 0 RCIM_Cubic>
         ]}}}
      Scale Curve=&FNTimeOfDayScalar(Night Scale=0.35, Morning Scale=0.72, Day Scale=0.65, Evening Scale=0.72)
     )
    )) {isSetParameters=true, module=SetVariables_#8}
   /SetVariables(Particles.scan color:LinearColor=&MakeLinearColorFromVectorAndFloat(
     "Vector (RGB)"=&MakeVectorFromLinearColorRGB(Color=<rgba 1 0 0.603 1>)
     "Float (Alpha)"=&FloatFromCurve(
      FloatCurve=DataInterface() {propertyValues {Curve {keys [
          {arriveTangent=-0.024, interpMode=RCIM_Cubic, leaveTangent=-0.024, tangentMode=RCTM_User, time=0, value=0}
          {arriveTangent=-0.3677, interpMode=RCIM_Cubic, leaveTangent=-0.3677, tangentMode=RCTM_User, time=0.1, value=1}
          {arriveTangent=-1.1111, interpMode=RCIM_Cubic, leaveTangent=-1.1111, time=0.4, value=0.35}
          <at 1 0 RCIM_Cubic>
         ]}}}
      Scale Curve=&FNTimeOfDayScalar(Night Scale=0.25, Morning Scale=0.25, Day Scale=0.25, Evening Scale=0.25)
     )
    )) {isSetParameters=true, module=SetVariables_#9}
   /SetVariables(Particles.Dyn:Vector4f=&MakeVector4(
     X=&FloatFromCurve(FloatCurve=DataInterface() {propertyValues {Curve {keys [
          {arriveTangent=-2.9077, interpMode=RCIM_Cubic, leaveTangent=-2.9077, tangentMode=RCTM_User, time=0, value=0.7}
          {arriveTangent=-1.991, interpMode=RCIM_Cubic, leaveTangent=-1.991, tangentMode=RCTM_User, time=0.3272, value=-0.2399}
          {arriveTangent=-0.376, interpMode=RCIM_Cubic, leaveTangent=-0.376, tangentMode=RCTM_User, time=1, value=-1}
         ]}}})
     Y=0
     Z=0
     W=0
    )) {isSetParameters=true, module=SetVariables_#10}
   /SetVariables(Particles.Grid color:LinearColor=&MakeLinearColorFromVectorAndFloat(
     "Vector (RGB)"=&MakeVectorFromLinearColorRGB(Color=<rgba 1 1 1 1>)
     "Float (Alpha)"=&FloatFromCurve(
      FloatCurve=DataInterface() {propertyValues {Curve {keys [<at 0 1>,<at 1 0>]}}}
      Scale Curve=&FN_FrontEndFloatModifier(
       Gameplay=&FNTimeOfDayScalar(Night Scale=1.5, Morning Scale=5, Day Scale=8, Evening Scale=4)
       Front End=1
      )
     )
    )) {isSetParameters=true, module=SetVariables_#11}
  ]
  rendererClasses [&NiagaraMeshRendererProperties, &NiagaraSpriteRendererProperties]
  renderers [
   {rendererClass=&NiagaraMeshRendererProperties, rendererIndex=0}
   {rendererClass=&NiagaraMeshRendererProperties, rendererIndex=1}
   {rendererClass=&NiagaraMeshRendererProperties, rendererIndex=2}
   {rendererClass=&NiagaraSpriteRendererProperties, rendererIndex=3}
  ]
 }
 Gameplay {
  emitterSpawnScript [/SetVariables(Emitter.Size=0.5) {isSetParameters=true, module=SetVariables_#2}]
  emitterUpdateScript [
   &EmitterState(Life Cycle Mode=::NewEnumerator1, Loop Duration=0.5, Recalculate Duration Each Loop=true, UseLoopDelay=false)
   &SpawnBurst_Instantaneous(Spawn Group=1, Use Spawn Probability=false, UseLoopCountLimit=false)
   &SpawnBurst_Instantaneous(Use Spawn Probability=false, UseLoopCountLimit=false) {module=SpawnBurst_Instantaneous001}
  ]
  particleSpawnScript [
   /SetVariables(Particles.VisibilityTag=&SpawnGroup()) {isSetParameters=true, module=SetVariables_#3}
   &InitializeParticle(
    Color=<rgba 0.5616 0.6 0.5671 1>
    Lifetime=&MergeHelper_FloatSwitch(
     0=&FN_FrontEndFloatModifier(Gameplay=1, Front End=0)
     1=&FN_FrontEndFloatModifier(Gameplay=0.4, Front End=0)
     2=?
     3=?
     Index=&SpawnGroup()
     Num=::NewEnumerator0
    )
    Mesh Scale Mode=::NewEnumerator3
    Mesh Uniform Scale=0.7
    Position Offset=<0 1 0>
    UsePositionOffset=false
    Write Material Random=true
   )
   /SetVariables(
    Particles.Glow Lifetime=0.3
    Particles.Glow Size=&Multiply_Vector2D_ByFloat(Float=$Emitter.Size, Vector2D=&Multiply_Vector2D_ByFloat(Float=80, Vector2D=<1 1>))
    Particles.GlowFacing:Vector3f=<0 0 1>
   ) {isSetParameters=true, module=SetVariables_#4}
   &InitialMeshOrientation(
    Mesh Axis To Orient=<0 0 1>
    Mesh Orientation Mode=::NewEnumerator2
    Orientation Coordinate Space=::Local
    Rotation=<5 0 0>
    Rotation Coordinate Space=::NewEnumerator0
    Use Rotation Vector=false
   ) {enabled=false}
  ]
  particleUpdateScript [
   &ParticleState()
   &SkeletalMeshLocation(
    Bone Sampling Mode=::NewEnumerator4
    MaskSpawnGroup=false
    Orientation Sampling=::NewEnumerator1
    Sampled Position Offset Coordinate Space:ENiagaraSkelMeshTransforms=::NewEnumerator0
    Skeletal Mesh=$System.Skeletal Mesh Reader
   )
   /SetVariables(Particles.Update Glow Age:NiagaraFloat=&Divide_Float(A=$Particles.Age, B=$Particles.Glow Lifetime)) {isSetParameters=true, module=SetVariables_#5}
   &ScaleMeshSize(Scale Factor=&VectorFromFloat(Value=&FloatFromCurve(
      FloatCurve=DataInterface() {propertyValues {Curve {keys [
          {arriveTangent=1.0263, interpMode=RCIM_Cubic, leaveTangent=1.0263, tangentMode=RCTM_User, time=0, value=1}
          {arriveTangent=0.4145, interpMode=RCIM_Cubic, leaveTangent=0.4145, tangentMode=RCTM_User, time=0.2, value=1.375}
          {arriveTangent=0.018, interpMode=RCIM_Cubic, leaveTangent=0.018, tangentMode=RCTM_User, time=0.9945, value=1.5}
         ]}}}
      Scale Curve=0.7
     )))
   /SetVariables(
    Particles.Update Glow Color:LinearColor=&MakeLinearColorFromVectorAndFloat(
     "Vector (RGB)"=&MakeVectorFromLinearColorRGB(Color=<rgba 0 1 0.6099 1>)
     "Float (Alpha)"=&FloatFromCurve(
      FloatCurve=DataInterface() {propertyValues {Curve {
         defaultValue=1
         keys [
          {interpMode=RCIM_Cubic, tangentMode=RCTM_User, time=0, value=0}
          {interpMode=RCIM_Cubic, tangentMode=RCTM_User, time=0.0724, value=1}
          {tangentMode=RCTM_User, time=1, value=0}
         ]
        }}}
      CurveIndex=$Particles.Update Glow Age
      Scale Curve=&FNTimeOfDayScalar(Night Scale=0.1, Morning Scale=0.1, Day Scale=0.15, Evening Scale=0.1)
     )
    )
    Particles.Update Glow DynamicParameter:Vector4f=<2 0 0 0>
    Particles.Update Glow Size:Vector2f=&Multiply_Vector2D(
     A=$Particles.Glow Size
     B=&Multiply_Vector2D_ByFloat(
      Float=&FloatFromCurve(
       FloatCurve=DataInterface() {propertyValues {Curve {
          defaultValue=0.367
          keys [
           {interpMode=RCIM_Cubic, tangentMode=RCTM_User, time=0, value=0.367}
           {arriveTangent=-0.1298, interpMode=RCIM_Cubic, leaveTangent=-0.1298, tangentMode=RCTM_User, time=1, value=3.8846}
          ]
         }}}
       CurveIndex=$Particles.Update Glow Age
       Scale Curve=1.5
      )
      Vector2D=<1.5 1.5>
     )
    )
   ) {isSetParameters=true, module=SetVariables_#6}
   /SetVariables(Particles.Scale002:Vector3f=&Multiply_Vector(
     A=<0.5 0.5 0.1>
     B=&VectorFromFloat(Value=&FloatFromCurve(FloatCurve=DataInterface() {propertyValues {Curve {keys [
           {arriveTangent=1.0263, interpMode=RCIM_Cubic, leaveTangent=1.0263, tangentMode=RCTM_User, time=0, value=1}
           {arriveTangent=0.3981, interpMode=RCIM_Cubic, leaveTangent=0.3981, tangentMode=RCTM_User, time=0.2, value=1.375}
           {arriveTangent=0.018, interpMode=RCIM_Cubic, leaveTangent=0.018, tangentMode=RCTM_User, time=0.9945, value=1.5}
          ]}}}))
    )) {isSetParameters=true, module=SetVariables_#7}
   &DynamicMaterialParameters(
    Index 0 Param 1=&FloatFromCurve(FloatCurve=DataInterface() {propertyValues {Curve {keys [
         {arriveTangent=-4.087, interpMode=RCIM_Cubic, leaveTangent=-4.087, tangentMode=RCTM_User, time=0, value=0.5}
         {arriveTangent=-1.721, interpMode=RCIM_Cubic, leaveTangent=-1.721, tangentMode=RCTM_User, time=0.1324, value=-0.0129}
         {arriveTangent=-0.376, interpMode=RCIM_Cubic, leaveTangent=-0.376, tangentMode=RCTM_User, time=1, value=-1}
        ]}}})
    Module.Index0Param0Write=true
    Module.Index0Param1Write=true
    Module.Index0Param2Write=true
    Module.Index0Param3Write=true
   )
   /SetVariables(Particles.Pixel color:LinearColor=&MakeLinearColorFromVectorAndFloat(
     "Vector (RGB)"=&MakeVectorFromLinearColorRGB(Color=<rgba 0.5 0 0.4489 2>)
     "Float (Alpha)"=&FloatFromCurve(
      FloatCurve=DataInterface() {propertyValues {Curve {keys [
          {arriveTangent=-0.024, interpMode=RCIM_Cubic, leaveTangent=-0.024, tangentMode=RCTM_User, time=0, value=0}
          {arriveTangent=-0.3677, interpMode=RCIM_Cubic, leaveTangent=-0.3677, tangentMode=RCTM_User, time=0.1, value=1}
          {arriveTangent=-1.5833, interpMode=RCIM_Cubic, leaveTangent=-1.5833, time=0.3, value=0.4}
          {arriveTangent=-0.3356, interpMode=RCIM_Cubic, leaveTangent=-0.3356, tangentMode=RCTM_User, time=0.7, value=0.05}
          <at 1 0 RCIM_Cubic>
         ]}}}
      Scale Curve=&FNTimeOfDayScalar(Night Scale=0.35, Morning Scale=0.72, Day Scale=0.65, Evening Scale=0.72)
     )
    )) {isSetParameters=true, module=SetVariables_#8}
   /SetVariables(Particles.scan color:LinearColor=&MakeLinearColorFromVectorAndFloat(
     "Vector (RGB)"=&MakeVectorFromLinearColorRGB(Color=<rgba 1 0 0.603 1>)
     "Float (Alpha)"=&FloatFromCurve(
      FloatCurve=DataInterface() {propertyValues {Curve {keys [
          {arriveTangent=-0.024, interpMode=RCIM_Cubic, leaveTangent=-0.024, tangentMode=RCTM_User, time=0, value=0}
          {arriveTangent=-0.3677, interpMode=RCIM_Cubic, leaveTangent=-0.3677, tangentMode=RCTM_User, time=0.1, value=1}
          {arriveTangent=-1.1111, interpMode=RCIM_Cubic, leaveTangent=-1.1111, time=0.4, value=0.35}
          <at 1 0 RCIM_Cubic>
         ]}}}
      Scale Curve=&FNTimeOfDayScalar(Night Scale=0.25, Morning Scale=0.25, Day Scale=0.25, Evening Scale=0.25)
     )
    )) {isSetParameters=true, module=SetVariables_#9}
   /SetVariables(Particles.Dyn:Vector4f=&MakeVector4(
     X=&FloatFromCurve(FloatCurve=DataInterface() {propertyValues {Curve {keys [
          {arriveTangent=-2.9077, interpMode=RCIM_Cubic, leaveTangent=-2.9077, tangentMode=RCTM_User, time=0, value=0.7}
          {arriveTangent=-1.991, interpMode=RCIM_Cubic, leaveTangent=-1.991, tangentMode=RCTM_User, time=0.3272, value=-0.2399}
          {arriveTangent=-0.376, interpMode=RCIM_Cubic, leaveTangent=-0.376, tangentMode=RCTM_User, time=1, value=-1}
         ]}}})
     Y=0
     Z=0
     W=0
    )) {isSetParameters=true, module=SetVariables_#10}
   /SetVariables(Particles.Grid color:LinearColor=&MakeLinearColorFromVectorAndFloat(
     "Vector (RGB)"=&MakeVectorFromLinearColorRGB(Color=<rgba 1 1 1 1>)
     "Float (Alpha)"=&FloatFromCurve(
      FloatCurve=DataInterface() {propertyValues {Curve {keys [<at 0 1>,<at 1 0>]}}}
      Scale Curve=&FN_FrontEndFloatModifier(
       Gameplay=&FNTimeOfDayScalar(Night Scale=1.5, Morning Scale=5, Day Scale=8, Evening Scale=4)
       Front End=1
      )
     )
    )) {isSetParameters=true, module=SetVariables_#11}
   &self:Slope_Position_02(
    Apply Position=true
    Apply Rotation=true
    Fallback Position:NiagaraPosition=&AddVectorToPosition(Position=$Particles.Initial.Position, Vector=<1 1 1>)
    Height Adjustment/Offset=3
    Slope Position:NiagaraPosition=$Particles.SkeletalMeshLocation.SampledPosition
    Trace Length=120
    Trace Origin Height=85
   ) {module=Slope_Position_02}
   &BlendSpace("Position Blend (World <-> Local)":NiagaraFloat=&MakeCustomFloatFromBool(
     Bool=&SetBoolByFloatComparison(A=&VectorLength(Vector=$User.OwnerLocalVelocity), B=0, Comparison Type=::NewEnumerator6)
     True Float=&FloatFromCurve(FloatCurve=DataInterface() {propertyValues {Curve {
         defaultValue=0
         keys [{interpMode=RCIM_Cubic, tangentMode=RCTM_User, time=0, value=1}, {interpMode=RCIM_Cubic, tangentMode=RCTM_User, time=0.15, value=0}]
        }}})
     False Float=&MakeCustomFloatFromBool(
      Bool=&SetBoolByFloatComparison(A=&VectorLength(Vector=$User.OwnerLocalVelocity), B=100, Comparison Type=::NewEnumerator6)
      True Float=&FloatFromCurve(FloatCurve=DataInterface() {propertyValues {Curve {
          defaultValue=0
          keys [{interpMode=RCIM_Cubic, tangentMode=RCTM_User, time=0, value=1}, {interpMode=RCIM_Cubic, tangentMode=RCTM_User, time=0.15, value=0}]
         }}})
      False Float=&FloatFromCurve(FloatCurve=DataInterface() {propertyValues {Curve {
          defaultValue=0
          keys [{interpMode=RCIM_Cubic, tangentMode=RCTM_User, time=0, value=1}, {interpMode=RCIM_Cubic, tangentMode=RCTM_User, time=1, value=0.5}]
         }}})
     )
     Translation Contribution:NiagaraFloat=1
     Rotation Contribution:NiagaraFloat=1
    )) {enabled=false, module=BlendSpace001}
  ]
  rendererClasses [&NiagaraMeshRendererProperties, &NiagaraSpriteRendererProperties]
  renderers [
   {rendererClass=&NiagaraMeshRendererProperties, rendererIndex=0}
   {rendererClass=&NiagaraMeshRendererProperties, rendererIndex=1}
   {rendererClass=&NiagaraMeshRendererProperties, rendererIndex=2}
   {rendererClass=&NiagaraSpriteRendererProperties, rendererIndex=3}
   {rendererClass=&NiagaraMeshRendererProperties, rendererIndex=4}
  ]
 }
}
summary {
 systemName=NS_Emote_EarPsych_floor
 emitters [
  {emitterName=Gameplay, rendererClasses [&NiagaraMeshRendererProperties, &NiagaraSpriteRendererProperties]}
  {emitterName=FE, rendererClasses [&NiagaraMeshRendererProperties, &NiagaraSpriteRendererProperties]}
 ]
 userVariables [
  User.bIsFrontEnd {type=NiagaraBool}
  User.bIsFrontEndPreview {type=NiagaraBool}
  User.BlinkIndex {type=NiagaraFloat, defaultValue=0, description=""}
  User.OwnerLocalVelocity {type=Vector3f, description="", defaultValue {struct=Vector3f, value=<0 0 0>}}
 ]
}
