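# Niagara system for the EarPsych emote that leaves glitchy light trails from bones or sockets on the character's skeletal mesh, presumably the hands.
# Hand_glitches spawns short-lived, spinning, falling mesh particles in two renderers, long and square glitches, with random colors and sizes scaled by camera distance.
# soft_light_trail builds a fading anim-trail ribbon that follows those bones, driven by user spawn-rate and color-lerp parameters.
# Time-of-day and front-end inputs adjust brightness and opacity.
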NiagaraSystem /BRCosmetics/Animation/Game/MainPlayer/Emotes/EarPsych/FX/NS_Emote_EarPsych_handtrails
dependencies {
 usedDataInterfaces [&NiagaraDataInterfaceCurve, &NiagaraDataInterfaceColorCurve, &NiagaraDataInterfaceSkeletalMesh]
 usedDynamicInputs [&FNTimeOfDayColor]
 usedModules [
  &self:SystemSpawnScript.SystemScriptSource.SystemScriptGraph.NiagaraNodeAssignment_0.SetVariables_#1
  &self:SystemSpawnScript.SystemScriptSource.SystemScriptGraph.NiagaraNodeAssignment_1.SetVariables_#2
  &SystemState
  &self:SystemSpawnScript.SystemScriptSource.SystemScriptGraph.NiagaraNodeAssignment_2.SetVariables_#3
  &self:SystemSpawnScript.SystemScriptSource.SystemScriptGraph.NiagaraNodeAssignment_3.SetVariables_#4
 ]
}
emitters {
 Hand_glitches {
  emitterSpawnScript [/SetVariables(
    Emitter.Long Glitch Opacity multiplier:NiagaraFloat=&FNTimeOfDayScalar(Night Scale=0.5, Morning Scale=0.47, Day Scale=0.65, Evening Scale=0.45)
    Emitter.Square Glitch Opacity mult:NiagaraFloat=&FNTimeOfDayScalar(Night Scale=0.6, Morning Scale=0.47, Day Scale=0.65, Evening Scale=0.45)
   ) {isSetParameters=true}]
  emitterUpdateScript [
   &EmitterState(Life Cycle Mode=::NewEnumerator1, Loop Behavior=::NewEnumerator1, Loop Duration=3, UseLoopDelay=false)
   &SpawnRate(SpawnGroup=1, SpawnRate=40, Use Spawn Probability=false) {module=SpawnRate001}
   &SpawnRate(SpawnRate=15, Use Spawn Probability=false)
  ]
  particleSpawnScript [
   /SetVariables(Particles.VisibilityTag=&SpawnGroup()) {isSetParameters=true}
   &PartitionParticles() {module=PartitionParticles001}
   &InitializeParticle(
    Color=<rgba 0.6385 0.8072 1 2>
    Lifetime Max=0.7
    Lifetime Min=0.35
    Lifetime Mode=::NewEnumerator1
    Sprite Rotation Mode=::NewEnumerator2
    Sprite Size Mode=::NewEnumerator3
    UsePositionOffset=false
    Write Material Random=true
   )
   &SkeletalMeshLocation(
    Bone / Socket Index=$Particles.PartitionParticles001.Partition
    Bone Sampling Mode=::NewEnumerator4
    MaskSpawnGroup=false
    Skeletal Mesh=$System.Skeletal Mesh
   )
   &ShapeLocation(MaskSpawnGroup=false, Override Seed=false, Sphere Radius=10, UseOwnerScale=true)
   &InheritVelocity(Speed Limit=true, UseSpeedThreshold=false)
   &DynamicMaterialParameters(
    Index 0 Param 1=&RandomRangeFloat(Minimum=0.3, Maximum=0)
    Index 0 Param 2=&RandomRangeFloat(Minimum=1, Maximum=2)
    Index 0 Param 3=&RandomRangeFloat(Minimum=-1, Maximum=1)
    Module.Index0Param0Write=true
    Module.Index0Param1Write=true
    Module.Index0Param2Write=true
    Module.Index0Param3Write=true
   )
   /SetVariables(
    Particles.Long Glitch Color:LinearColor=&RandomRangeLinearColor(Minimum=<rgba 0 1 0.6099 1>, Maximum=<rgba 0.0004 0.8856 1 1>)
    Particles.Long Glitch Size:Vector3f=&RandomRangeVector(Minimum=<0.25 2.5 0.25>, Maximum=<2 0.6294 8.9034>)
   ) {isSetParameters=true}
   /SetVariables(
    Particles.Square Glitch Color:LinearColor=&RandomRangeLinearColor(Minimum=<rgba 1 0 0.8733 1>, Maximum=<rgba 0.5061 0 1 1>)
    Particles.Square Glitch Size:Vector3f=&RandomRangeVector(Minimum=<2 2 2>, Maximum=<5 5 5>)
   ) {isSetParameters=true}
  ]
  particleUpdateScript [
   &ParticleState()
   &UpdateMeshOrientation(Rotation Rate=7.8512, Rotation Vector=<0 0 1>)
   &ScaleSpriteSize(Uniform Curve Sprite Scale=DataInterface() {propertyValues {Curve {keys [{tangentMode=RCTM_User, time=0, value=0.05}, <at 0.2 0.85>, <at 0.5 0.95>, {tangentMode=RCTM_User, time=1, value=0}]}}})
   &GravityForce(Gravity=<0 0 -200>)
   &Drag(Drag=3, Use Linear Drag=true, Use Rotational Drag=true)
   &MeshRotationForce(Rotation=<0 0 8.7256>)
   &SolveForcesAndVelocity(Clamp Velocity=false, Limit Acceleration=false)
   /SetVariables(
    Particles.Square Glitch Color Update:LinearColor=&MakeLinearColorFromVectorAndFloat("Vector (RGB)"=&MakeVectorFromLinearColorRGB(Color=$Particles.Square Glitch Color), "Float (Alpha)"=1)
    Particles.Square Glitch DynParam 1:Vector4f=&MakeVector4(
     X=0
     Y=&RandomRangeFloat(Minimum=-1, Maximum=1)
     Z=&FloatFromCurve(FloatCurve=DataInterface() {propertyValues {Curve {keys [
          <at 0 2 RCIM_Cubic>
          {arriveTangent=-6.2285, interpMode=RCIM_Cubic, leaveTangent=-6.2285, tangentMode=RCTM_User, time=0.5612, value=1.0841}
          {arriveTangent=-0.8021, interpMode=RCIM_Cubic, leaveTangent=-0.8021, tangentMode=RCTM_User, time=1, value=0}
         ]}}})
     W=0
    )
    Particles.Square Glitch DynParam 2:Vector4f=&MakeVector4(
     X=&RandomRangeFloat(Minimum=-1, Maximum=1)
     Y=0
     Z=0
     W=&FloatFromCurve(FloatCurve=DataInterface() {propertyValues {Curve {keys [
          {arriveTangent=4.1785, interpMode=RCIM_Cubic, leaveTangent=4.1785, tangentMode=RCTM_User, time=0, value=0}
          {arriveTangent=0.9995, interpMode=RCIM_Cubic, leaveTangent=0.9995, tangentMode=RCTM_User, time=0.2223, value=0.6825}
          {arriveTangent=-0.0405, interpMode=RCIM_Cubic, leaveTangent=-0.0405, tangentMode=RCTM_User, time=1, value=1}
         ]}}})
    )
    Particles.Square Glitch Scale Update:Vector3f=&Multiply_Vector(
     A=$Particles.Initial.Square Glitch Size
     B=&Multiply_VectorByFloat(
      Vector=&RandomRangeVector(Minimum=<0.05 0.05 0.05>, Maximum=<0.2 0.2 0.2>)
      Float=&Multiply_Float(
       A=0.08
       B=&FloatFromCurve(
        FloatCurve=DataInterface() {propertyValues {Curve {keys [
            {arriveTangent=5.4245, interpMode=RCIM_Cubic, leaveTangent=5.4245, tangentMode=RCTM_User, time=0, value=5.7247}
            {arriveTangent=9.4279, interpMode=RCIM_Cubic, leaveTangent=9.4279, time=0.0625, value=10.3835}
            {arriveTangent=-31.0729, interpMode=RCIM_Cubic, leaveTangent=-31.0729, time=0.15, value=7.1389}
            {arriveTangent=-20.9259, interpMode=RCIM_Cubic, leaveTangent=-20.9259, time=0.2702, value=3.9295}
            <at 0.3 4>
            <at 0.3956 5.8879>
            {arriveTangent=0.1532, interpMode=RCIM_Cubic, leaveTangent=0.1532, tangentMode=RCTM_User, time=0.5, value=2.875}
            <at 0.8 2.5>
            <at 0.9 2>
            {arriveTangent=-0.0405, interpMode=RCIM_Cubic, leaveTangent=-0.0405, tangentMode=RCTM_User, time=1, value=0.5}
           ]}}}
        Scale Curve=&Multiply_Float(
         A=0.8
         B=&ClampFloat(
          Float=&FN_LOD_Distance_Remap(Start Distance=200, End Distance=400, Minimum Output=0.5, Maximum Output=1)
          Min=0.5
         )
        )
       )
      )
     )
    )
   ) {isSetParameters=true}
   /SetVariables(
    Particles.Long Glitch Color Update:LinearColor=&MakeLinearColorFromVectorAndFloat(
     "Vector (RGB)"=&MakeVectorFromLinearColorRGB(Color=$Particles.Long Glitch Color)
     "Float (Alpha)"=&FN_FrontEndFloatModifier(
      Gameplay=&FNTimeOfDayScalar(Night Scale=0, Morning Scale=0, Day Scale=5, Evening Scale=0)
      Front End=0
     )
    )
    Particles.Long Glitch Dynam 1:Vector4f=&MakeVector4(
     X=0
     Y=&RandomRangeFloat(Minimum=-1, Maximum=1)
     Z=&FloatFromCurve(FloatCurve=DataInterface() {propertyValues {Curve {keys [
          <at 0 2 RCIM_Cubic>
          {arriveTangent=-6.2285, interpMode=RCIM_Cubic, leaveTangent=-6.2285, tangentMode=RCTM_User, time=0.5612, value=1.0841}
          {arriveTangent=-0.8021, interpMode=RCIM_Cubic, leaveTangent=-0.8021, tangentMode=RCTM_User, time=1, value=0}
         ]}}})
     W=0
    )
    Particles.Long Glitch Dynam 2:Vector4f=&MakeVector4(
     X=0
     Y=0
     Z=0
     W=&FloatFromCurve(FloatCurve=DataInterface() {propertyValues {Curve {keys [
          {arriveTangent=4.1785, interpMode=RCIM_Cubic, leaveTangent=4.1785, tangentMode=RCTM_User, time=0, value=0}
          {arriveTangent=0.9995, interpMode=RCIM_Cubic, leaveTangent=0.9995, tangentMode=RCTM_User, time=0.2223, value=0.6825}
          {arriveTangent=-0.0405, interpMode=RCIM_Cubic, leaveTangent=-0.0405, tangentMode=RCTM_User, time=1, value=1}
         ]}}})
    )
    Particles.Long Glitch Scale:Vector3f=&Multiply_Vector(
     A=$Particles.Initial.Long Glitch Size
     B=&Multiply_VectorByFloat(
      Vector=<0.5 0.5 0.2>
      Float=&Multiply_Float(
       A=0.08
       B=&FloatFromCurve(
        FloatCurve=DataInterface() {propertyValues {Curve {keys [
            <at 0.0024 4.7657 RCIM_Cubic>
            {arriveTangent=14.525, interpMode=RCIM_Cubic, leaveTangent=14.525, tangentMode=RCTM_Break, time=0.0625, value=7.4517}
            {arriveTangent=-14.3776, interpMode=RCIM_Cubic, leaveTangent=-14.3776, tangentMode=RCTM_Break, time=0.15, value=6.2624}
            {arriveTangent=0.1532, interpMode=RCIM_Cubic, leaveTangent=0.1532, tangentMode=RCTM_Break, time=0.5, value=2.875}
            {interpMode=RCIM_Cubic, tangentMode=RCTM_Break, time=0.8185, value=2.3106}
            {arriveTangent=-0.0405, interpMode=RCIM_Cubic, leaveTangent=-0.0405, tangentMode=RCTM_Break, time=0.9185, value=0.8106}
           ]}}}
        Scale Curve=&Multiply_Float(
         A=0.5
         B=&ClampFloat(
          Float=&FN_LOD_Distance_Remap(Start Distance=200, End Distance=400, Minimum Output=0.5, Maximum Output=1)
          Min=0.5
         )
        )
       )
      )
     )
    )
   ) {isSetParameters=true}
   /SetVariables(
    Particles.box rotation:NiagaraFloat=&Multiply_Float(A=&RandomRangeFloat(Minimum=0, Maximum=1), B=$Engine.DeltaTime)
    Particles.irridescent box:Vector3f=<0.1 1 1>
   ) {isSetParameters=true}
  ]
  rendererClasses [&NiagaraMeshRendererProperties]
  renderers [
   {rendererClass=&NiagaraMeshRendererProperties, rendererIndex=0}
   {rendererClass=&NiagaraMeshRendererProperties, rendererIndex=1}
  ]
 }
 soft_light_trail {
  emitterUpdateScript [
   &EmitterState()
   &FN_SpawnAnimTrail(Spawn Mode:ENiagaraSkeletalSpawning=::NewEnumerator0)
  ]
  particleSpawnScript [
   &FN_InitializeAnimTrailBoneIndecies()
   &FN_InitializeAnimTrail(
    AnimTrailWidth=0.5
    Lifetime=&Multiply_Float(
     A=&FloatFromCurve(
      FloatCurve=DataInterface() {propertyValues {Curve {
         defaultValue=1
         keys [{interpMode=RCIM_Cubic, tangentMode=RCTM_User, time=0, value=1}, {interpMode=RCIM_Cubic, tangentMode=RCTM_User, time=0.4905, value=-0.0601}]
        }}}
      Scale Curve=0.9
     )
     B=$User.AnimTrail_SpawnRate
    )
    SkeletalMesh=$System.Skeletal Mesh
   )
   &PartitionParticles(Write Partition to Ribbon ID=::NewEnumerator0)
   &ApplyInitialForces()
   &SkeletalMeshLocation(
    Bone / Socket Index=$Particles.PartitionParticles.Partition
    Bone Sampling Mode=::NewEnumerator4
    MaskSpawnGroup=false
    Skeletal Mesh=$System.Skeletal Mesh
   )
  ]
  particleUpdateScript [
   &ParticleState()
   &FN_UpdateAnimTrail(SkeletalMesh=$System.Skeletal Mesh)
   &ScaleRibbonWidth(Ribbon Width Scale=&FloatFromCurve(FloatCurve=DataInterface() {propertyValues {Curve {
        defaultValue=1
        keys [
         {arriveTangent=0.0733, interpMode=RCIM_Cubic, leaveTangent=0.0733, tangentMode=RCTM_User, time=0, value=1}
         {arriveTangent=-0.0405, interpMode=RCIM_Cubic, leaveTangent=-0.0405, tangentMode=RCTM_User, time=0.7, value=0.3}
        ]
       }}}))
   &InheritSourceMovement(ApplyForce=false, ApplyPosition=true, ApplyPositionFalloff=false)
   &Drag(Drag=5, Use Linear Drag=true, Use Rotational Drag=false)
   &SolveForcesAndVelocity(Clamp Velocity=false, Limit Acceleration=false)
   /SetVariables(Particles.EnergyColor:LinearColor=/DealershipSharedLibrary/FX/Niagara/DynamicInputs/MultiplyLinearColorByVectorAndFloat(
     Color to Scale:LinearColor=&ColorFromCurve(ColorCurve=DataInterface() {propertyValues {
        AlphaCurve {keys [<at 0 0.5 RCIM_Cubic>,<at 0.9052 0 RCIM_Cubic>]}
        BlueCurve {keys [{arriveTangent=-0.4094, interpMode=RCIM_Cubic, time=0.2155, value=1}, <at 0.9527 0.6099>]}
        GreenCurve {keys [{arriveTangent=0.2163, interpMode=RCIM_Cubic, time=0.2155, value=0.1885}, <at 0.9527 1>]}
        RedCurve {keys [{arriveTangent=-0.644, interpMode=RCIM_Cubic, time=0.2155, value=0.898}, <at 0.9527 0>]}
       }})
     RGB Scale:Vector3f=<1 1 1>
     Alpha Scale:NiagaraFloat=&FloatFromCurve(
      FloatCurve=DataInterface() {propertyValues {Curve {keys [{arriveTangent=-3.0806, interpMode=RCIM_Cubic, leaveTangent=-3.0806, tangentMode=RCTM_User, time=0, value=1}, <at 0.85 0.2071 RCIM_Cubic>]}}}
      Scale Curve=&Lerp_Float(
       A=0
       B=&FNTimeOfDayScalar(Night Scale=5, Morning Scale=5, Day Scale=5, Evening Scale=5)
       Alpha=$User.EnergyTrail_Color_Lerp
      )
     )
    )) {isSetParameters=true}
   /SetVariables(Particles.Energy ribbon width:NiagaraFloat=&FloatFromCurve(
     FloatCurve=DataInterface() {propertyValues {Curve {keys [
         {arriveTangent=-4.0337, interpMode=RCIM_Cubic, tangentMode=RCTM_Break, time=0, value=1}
         {arriveTangent=-4.0337, interpMode=RCIM_Cubic, tangentMode=RCTM_Break, time=1, value=0}
        ]}}}
     Scale Curve=20
    )) {isSetParameters=true}
  ]
  rendererClasses [&NiagaraRibbonRendererProperties]
  renderers [{rendererClass=&NiagaraRibbonRendererProperties}]
 }
}
summary {
 systemName=NS_Emote_EarPsych_handtrails
 emitters [
  {emitterName=soft_light_trail, rendererClasses [&NiagaraRibbonRendererProperties]}
  {emitterName=Hand_glitches, rendererClasses [&NiagaraMeshRendererProperties]}
 ]
 userVariables [
  User.Anim_Curve_SpawnRate_Ribbon {type=NiagaraFloat, defaultValue=0, description=""}
  User.AnimTrail_SpawnRate {type=NiagaraFloat, defaultValue=0, description=""}
  User.bIsFrontEnd {type=NiagaraBool}
  User.bIsFrontEndPreview {type=NiagaraBool}
  User.EnergyTrail_Color_Lerp {type=NiagaraFloat, defaultValue=1, description=""}
  User.FrostyMistAlpha {type=NiagaraFloat, defaultValue=0, description=""}
  User.FrostyMistTrail_Rate {type=NiagaraFloat, defaultValue=1, description=""}
  User.RibbonIDOffset {type=NiagaraInt32}
 ]
}
